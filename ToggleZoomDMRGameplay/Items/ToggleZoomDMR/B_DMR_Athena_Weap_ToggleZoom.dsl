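; Weapon Blueprint for the DMR, whose scope toggles between 2x and 3x zoom while aiming down sights.
; The zoom target-mode component drives camera swaps, zoom gameplay tags, audio and a timeline lerping scope offset plus cached post-process and dial materials.
; Local firing plays camera shake and timeline recoil timed to fire rate, and flags bullets as predicted.
; A scalable float decides whether leaving ADS resets zoom.
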(event EventBeginPlay
  ;[1 1 1] Cache firing rate
  (|CallParentFunction)
  (bind _outrow (Fortnite|GetRangedWeaponStatsRow (Data|GetWeaponDataWeaponStatHandle)))
  (Variables|Default|SetFiring%20Rate (.FiringRate (Utilities|Struct|BreakFortRangedWeaponStats _outrow)))
  (Networking|CanExecuteCosmeticEvents
    (:True
      (CallFunction|UpdateTargetMode)
      ;[0.3816 0.002553 1] Reset zoom amount when ADS is exited. 0 = reset
      (if (not (> (ScalableFloat|GetValueatLevel (Variables|Default|GetSF_SaveZoomState)) 0))
        (Default|BindEventtoOnAimDownSightsChanged (delegate Custom|ResetZoom))))))

(event Custom|ResetZoom
  (params (AimDownSights :type Boolean))
  ;[0.3816 0.002553 1] Reset zoom amount when ADS is exited. 0 = reset
  (if (not AimDownSights)
    (Class|FortWeaponComponentTargetModeSelector|ServerSetTargetMode (Variables|Default|GetFortWeaponComponent_TargetModeSelector_Zoom) 0)
    (Variables|Scope|SetScope%20Camera%20Offset%20Amount (Variables|Default|GetDefaultCameraScopeOffset))))

(event EventOnPlayWeaponFireFX
  (params (bPersistentFire :type Boolean) (bSecondaryFire :type Boolean))
  (bind _fireshake (Variables|B_DMR_Athena_Weap_ToggleZoom|GetFireShake))
  ;[1 1 1] Recoil Scope Mesh Animation
  (|CallParentFunction bPersistentFire bSecondaryFire)
  (if (Pawn|IsLocallyControlled (Variables|Actor|GetInstigator))
    (Components|Timeline|SetTimelineLength _fireshake (/ 1.0 (Variables|Default|GetFiring%20Rate)))
    (Components|Timeline|SetLooping _fireshake false)
    (CallFunction|RecoilCamShake)
    (Variables|Default|SetFiring true)
    (if (Sounds|IsQuartzGunfireEnabled)
      (bind _returnvalue (Utilities|Time|SetTimerbyEvent (EventDispatchers|CreateEvent :!event Custom|RecoilCamShake) (/ 1.0 (Variables|Default|GetFiring%20Rate)) true))
      (Variables|Default|SetRecoilTimerHandle _returnvalue)
      (bind (_direction _newtrack_0) (Timeline|FireShake PlayFromStart)
        (:Update
          (Variables|Default|SetScope%20Camera%20Recoil%20Offset (Math|Float|Lerp 0.0 (Variables|Default|GetMax%20Scope%20Camera%20Recoil%20Offset) _newtrack_0))
          (Variables|Default|SetXRecoilOffset (Math|Float|Lerp 0.0 (Math|Random|RandomFloatinRange -0.8 0.8) _newtrack_0))
          (Variables|Default|SetYRecoilOffset (Math|Float|Lerp 0.0 0.3 _newtrack_0))))
      (else
        (Timeline|FireShake PlayFromStart)))))

(event EventOnStopWeaponFireFX
  ;[1 1 1] Recoil Scope Mesh Animation
  (|CallParentFunction))

(event Custom|RecoilCamShake
  ;[1 1 1] Recoil Scope Mesh Animation
  (if (Camera|GetIsTargeting)
    (CameraShakes|StartCameraShake :self (Variables|/Script/Engine.PlayerController|GetPlayerCameraManager (Fortnite|GetFortPlayerControllerfromActor (Game|GetInstigator))) :Scale 1.0 :PlaySpace "CameraLocal")))

(event OnTargetModeChanged(FortWeaponComponent_TargetModeSelector_Zoom)
  (params (TargetMode :type Integer))
  ;[0.08209 1 0.8601] Change Camera Mode when toggle input is pressed during ADS
  (Networking|CanExecuteCosmeticEvents
    (:True
      (Variables|Default|SetTarget%20Mode TargetMode)
      (CallFunction|UpdateTargetMode)
      (Default|CallOnZoomChanged (CallFunction|GetIsZoomed)))))

(event Custom|UpdateTargetMode
  (bind _fortweaponcomponent_targetmodeselector_zoom (Variables|Default|GetFortWeaponComponent_TargetModeSelector_Zoom))
  ;[0.08209 1 0.8601] Change Camera Mode when toggle input is pressed during ADS
  (CallFunction|UpdateZoomTags)
  (Camera|SetTargeting3PCamera (Utilities|Select :Index (Variables|Default|GetTarget%20Mode) :Option%200 (Variables|Default|GetDefaultTargetingCam) :Option%201 (Variables|Default|GetZoomedTargetingCam)))
  (CallFunction|PlayToggleZoomAudio)
  (if (== (Variables|Default|GetTarget%20Mode) 0)
    (bind (_direction _newtrack_0) (Timeline|Timeline PlayFromStart)
      (:Update
        (bind _output_get (Variables|Default|SetScopeSizeConstant (Math|Float|Lerp 4.811374 4.548876 _newtrack_0)))
        (Class|FortWeaponComponentTargetModeSelectorZoom|LerpScopeMaterialParameters _fortweaponcomponent_targetmodeselector_zoom _newtrack_0)
        (Variables|Scope|SetScope%20Camera%20Offset%20Amount (+ (Variables|Default|GetScope%20Camera%20Recoil%20Offset) (Class|FortWeaponComponentTargetModeSelectorZoom|GetScopeCameraOffset _fortweaponcomponent_targetmodeselector_zoom _output_get (Variables|Default|GetDefaultFOV) (Variables|Default|GetDefaultCameraScopeOffset))))))
    (else
      (Timeline|Timeline ReverseFromEnd))))

(event FortWeapon|EventOnSpawnProjectile
  (params (SpawnProjectile :type Fort%20Projectile%20Base%20Object%20Reference))
  (if (Pawn|IsLocallyControlled (Variables|Actor|GetInstigator))
    (bind _asb_prj_bullet_dmr (Utilities|Casting|CastToB_Prj_Bullet_DMR SpawnProjectile))
    (Variables|/ToggleZoomDMRGameplay/Items/ToggleZoomDMR/B_Prj_Bullet_DMR.B_Prj_Bullet_DMR_C|SetIsPredicted true _asb_prj_bullet_dmr)))

(event Custom|PlayToggleZoomAudio
  ;[0 0 0] Toggle zoom audio
  (bind _asfort_player_pawn_athena (Utilities|Casting|CastToFortPlayerPawnAthena (Game|GetInstigator)))
  (if (Pawn|IsLocallyViewed _asfort_player_pawn_athena)
    (if (FortWeapon|IsTargeting)
      (Audio|PlaySoundatLocation (Utilities|Select :Index (Variables|Default|GetTarget%20Mode) :Option%200 (Variables|Default|GetToggleZoomOut) :Option%201 (Variables|Default|GetToggleZoomIn)) (Transformation|GetActorLocation)))))

(event EventOnSetTargeting
  (params (bNewIsTargeting :type Boolean))
  (|CallParentFunction bNewIsTargeting)
  (CallFunction|UpdateZoomTags))

(event EventUnhideFirstPersonStuffPart2
  (params (Which%20Call :type Integer))
  (bind _scopecomponent (CallFunction|GetScopeComp))
  (Utilities|IsValid _scopecomponent
    (:Is%20Valid
      (Rendering|SetRenderCustomDepth _scopecomponent true)
      (Rendering|SetCustomDepthStencilValue _scopecomponent 3)
      (CallFunction|CachePostProcessMaterialInstances)
      (Class|FortWeaponComponentTargetModeSelectorZoom|LerpScopeMaterialParameters (Variables|Default|GetFortWeaponComponent_TargetModeSelector_Zoom) (- 1.0 (Math|Conversions|ToFloat(Integer) (Variables|Default|GetTarget%20Mode))))
      (|CallParentFunction Which%20Call))))

(event EventHideFirstPersonStuff
  (bind _scopecomponent (CallFunction|GetScopeComp))
  (Utilities|IsValid _scopecomponent
    (:Is%20Valid
      (Rendering|SetRenderCustomDepth _scopecomponent false)
      (|CallParentFunction))))

(fn CachePostProcessMaterialInstances
  (bind (_weight _object) (Utilities|Struct|BreakWeightedBlendable (Utilities|Array|Get(aref) (Scope|GetScopePostProcessMaterial))))
  (bind (_weight_2 _object_2) (Utilities|Struct|BreakWeightedBlendable (Utilities|Array|Get(aref) (Scope|GetScopePostProcessMaterial) 2)))
  (bind (_weight_3 _object_3) (Utilities|Struct|BreakWeightedBlendable (Utilities|Array|Get(aref) (Scope|GetScopePostProcessMaterial) 1)))
  (bind _returnvalue (Scope|GetScopePostProcessMaterial))
  (Utilities|FlowControl|Sequence
    (:then_0
      ;[0.9703 0.4037 1] Cach scope post process materials
      (Utilities|FlowControl|Sequence
        (:then_0
          (bind _asmaterial_instance_dynamic (Utilities|Casting|CastToMaterialInstanceDynamic _object)
            (:then
              (Variables|Default|SetZoomPostDynamicInstanceA _asmaterial_instance_dynamic))
            (:CastFailed
              _)))
        (:then_1
          (bind _asmaterial_instance_dynamic_2 (Utilities|Casting|CastToMaterialInstanceDynamic _object_3)
            (:then
              (Variables|Default|SetZoomPostDynamicInstanceB _asmaterial_instance_dynamic_2))
            (:CastFailed
              _)))
        (:then_2
          (bind _asmaterial_instance_dynamic_3 (Utilities|Casting|CastToMaterialInstanceDynamic _object_2)
            (:then
              (Variables|Default|SetZoomPostDynamicInstanceC _asmaterial_instance_dynamic_3))
            (:CastFailed
              _))))
      (return false))
    (:then_1
      ;[0.5099 0.1258 0.6667] Create and cache scope mesh dynamic material
      (bind _scopecomponent (CallFunction|GetScopeComp))
      (Utilities|IsValid _scopecomponent
        (:Is%20Valid
          (Utilities|IsValid (Variables|/Script/Engine.StaticMeshComponent|GetStaticMesh _scopecomponent)
            (:Is%20Valid
              (bind _returnvalue_2 (Rendering|Material|CreateDynamicMaterialInstance (StaticMesh|GetMaterial (Variables|/Script/Engine.StaticMeshComponent|GetStaticMesh _scopecomponent) 0) "None" "None" :!class /Script/Engine.KismetMaterialLibrary))
              (bind _output_get (Variables|Default|SetScopeMID _returnvalue_2))
              (Utilities|IsValid _output_get
                (:Is%20Valid
                  (Rendering|Material|SetMaterial _scopecomponent 0 _output_get))
                (:Is%20Not%20Valid
                  (return false))))))))
    (:then_2
      ;[0.4691 0.28 1] Setup interpolation settings for materials during zoom transition
      (Variables|/Script/FortniteGame.FortWeaponComponent_TargetModeSelector_Zoom|SetMaterialLerpInfo (Utilities|Array|MakeArray (Utilities|Struct|MakeMaterialParamterLerpInfo (Variables|Default|GetZoomPostDynamicInstanceA) "percent" 0.032 0.005) (Utilities|Struct|MakeMaterialParamterLerpInfo (Variables|Default|GetZoomPostDynamicInstanceB) "percent" 0.032 0.005) (Utilities|Struct|MakeMaterialParamterLerpInfo (Variables|Default|GetZoomPostDynamicInstanceC) "ZoomFactor" 1.4 1.0) (Utilities|Struct|MakeMaterialParamterLerpInfo (Variables|Default|GetScopeMID) "ScopeRotateDialAngle" 6.0 3e+01)) (Variables|Default|GetFortWeaponComponent_TargetModeSelector_Zoom)))
    (:then_3
      (return true))))

(fn GetIsZoomed
  (return (> (Variables|Default|GetTarget%20Mode) 0)))

(fn initScope
  (|CallParentFunction)
  (CallFunction|CachePostProcessMaterialInstances))

(fn RemoveZoomTags
  (Ability|GameplayEffect|RemoveLooseGameplayTags (Variables|Actor|GetInstigator) (LocalVariables|GetZoomTags)))

(fn UpdateZoomTags
  (CallFunction|RemoveZoomTags)
  (if (FortWeapon|IsTargeting)
    (Ability|GameplayEffect|AddLooseGameplayTags (Variables|Actor|GetInstigator) (Utilities|Select :Index (Variables|Default|GetTarget%20Mode) :Option%200 (LocalVariables|Get2xZoomTag) :Option%201 (LocalVariables|Get3xZoomTag)))))

(fn ConstructionScript
  (|CallParentFunction)
  (Variables|Default|SetScopeSizeConstant (* (Variables|Default|GetDefaultCameraScopeOffset) (Math|Trig|Tan(Degrees) (* (Variables|Default|GetDefaultFOV) 0.5)))))
