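# Post-process material for the Toggle Zoom DMR that applies a masked, one-axis Gaussian blur to the scene.
# A Custom HLSL node takes TemporalAA-dithered samples, with the horv static switch choosing the horizontal or vertical axis.
# Blur offsets scale by a soft screen-aligned circle mask and a smoothed custom-depth range.
# The mobile shading path passes the scene through unblurred, and the RadiusPixels input is unused in favor of percent.
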Material /ToggleZoomDMRGameplay/Items/ToggleZoomDMR/Materials/M_PP_ToggleZoom
graph {
 properties {bIsBlendable=false, materialDomain=MD_PostProcess, pixelDepthOffsetMode=PDOM_Legacy, usedShadingModels=Default Lit}
 nodes [
  Divide(A=SceneTexture() {sceneTextureId=PPI_SceneDepth}.Color) {constB=1000}
  MaterialExpressionScalarParameter_0 = ScalarParameter() {defaultValue=14, parameterName=depthMin}
  MaterialExpressionScalarParameter_1 = ScalarParameter() {defaultValue=14.5, parameterName=depthMax}
  n1 = MaterialFunctionCall!(
   Input=SceneTexture() {sceneTextureId=PPI_CustomDepth}.Color
   Input Low=@MaterialExpressionScalarParameter_0
   Input High=@MaterialExpressionScalarParameter_1
  ) {materialFunction=/Engine/Functions/Engine_MaterialFunctions03/Math/RemapValueRangeNormalized}
  Frac(None=@n1.Result)
  OneMinus(None=Saturate(None=@n1.Result))
  MaterialExpressionScalarParameter_3 = ScalarParameter() {defaultValue=10, parameterName=Samples}
  MaterialExpressionScalarParameter_4 = ScalarParameter() {defaultValue=49.5, parameterName=RadiusPixels}
  MaterialExpressionMaterialFunctionCall_8 = MaterialFunctionCall() {materialFunction=&DitherTemporalAA}
  MaterialExpressionSceneTexture_2 = SceneTexture() {sceneTextureId=PPI_PostProcessInput0}
  MaterialExpressionStaticSwitchParameter_0 = StaticSwitchParameter(True=Constant() {r=1}, False=Constant()) {parameterName=horv}
  n2 = ScalarParameter() {defaultValue=0.05, parameterName=blurWidth}
  MaterialExpressionMaterialFunctionCall_10 = MaterialFunctionCall(
   Input=MaterialFunctionCall(
    UVs=MaterialFunctionCall() {materialFunction=&ScreenAlignedUVs}."X Scale to Ratio, Y 100%"
    Size=Constant() {r=1}
   ) {materialFunction=&MF_UI_Circle}.SDF
   Input Low=Multiply(A=@n2) {constB=-1}
   Input High=@n2
  ) {materialFunction=&RemapValue_0-1}
  MaterialExpressionScalarParameter_2 = ScalarParameter() {defaultValue=0.024, parameterName=percent}
  n3 = Custom!(
   Samples=@MaterialExpressionScalarParameter_3
   RadiusPixels=@MaterialExpressionScalarParameter_4
   Dither=@MaterialExpressionMaterialFunctionCall_8.Result
   SceneTexture=@MaterialExpressionSceneTexture_2.Color
   switchDirection=@MaterialExpressionStaticSwitchParameter_0
   mask=@MaterialExpressionMaterialFunctionCall_10.Result
   minDepth=@MaterialExpressionScalarParameter_0
   maxDepth=@MaterialExpressionScalarParameter_1
   percent=@MaterialExpressionScalarParameter_2
  ) {
   code="\r\nstatic const int SceneTextureID = 14;\r\nstatic const int StencilTextureID = 25;\r\nstatic const int CustomDepthTextureID = 13;\r\n\r\n\r\nstatic const float2 direction[2] = {float2(1,0),float2(0,1)};\r\nstatic const int sd = switchDirection;\r\n\r\nfloat2 TexelSize = View.BufferSizeAndInvSize.xy;\r\n\r\n\r\n//static float aspect[2] = {TexelSize.x/TexelSize.y,TexelSize.y/TexelSize.x};\r\nstatic float aspect[2] = {TexelSize.y/TexelSize.x,TexelSize.x/TexelSize.y};\r\n\r\nfloat2 UV2 = GetDefaultSceneTextureUV(Parameters, StencilTextureID);\r\n\r\nfloat2 UV = GetDefaultSceneTextureUV(Parameters, SceneTextureID);\r\nfloat2 UVDepth = GetDefaultSceneTextureUV(Parameters, CustomDepthTextureID);\r\n//UVDepth -= 0.5;\r\n//UVDepth *= 1.002;\r\n//UVDepth += 0.5;\r\n\r\n\r\nfloat3 PixelSum = float3(0,0,0);\r\nfloat WeightSum = 0;\r\n\r\n//float3 PixelSum = SceneTextureLookup(UV, SceneTextureID, 0).rgb;\r\n//float WeightSum = 1;\r\n\r\n\r\nfloat stencil = SceneTextureLookupCommon(Parameters, UV2, StencilTextureID, 1).r;\r\nfloat3 PixelColorOrig = SceneTextureLookupCommon(Parameters, UV, SceneTextureID, 0).rgb;\r\n\r\n\r\n\r\nstruct UtilityFunctions{\r\n    static  float mapRange(float s, float a1, float a2, float b1, float b2)\r\n    {\r\n        return b1 + (s-a1)*(b2-b1)/(a2-a1);\r\n    }\r\n    static  float mapRangeClamped(float s, float a1, float a2, float b1, float b2)\r\n    {\r\n        return b1 + (s-a1)*(b2-b1)/(a2-a1);\r\n    }\r\n    static  float mapRangeNormalized(float s, float a1, float a2)\r\n    {\r\n        return saturate((s-a1)/(a2-a1));\r\n    }\r\n};\r\n\r\n\r\n\r\nfloat depthMask = smoothstep(0, 1, UtilityFunctions::mapRangeNormalized(SceneTextureLookupCommon(Parameters, UVDepth, CustomDepthTextureID, 0).r, minDepth, maxDepth));\r\n//shifts samples into equal samples when used with dither so that you dont get overlaps at the boundaries\r\nfloat halfTexelOffsetFactor = (Samples*2);\r\nfloat halfTexelOffsetFactorFloat = halfTexelOffsetFactor/(halfTexelOffsetFactor+1);\r\n\r\nfor (int y = -Samples; y <= Samples; y++){\r\n\r\n    float norm = (y + Dither) / (Samples) * halfTexelOffsetFactorFloat;\r\n\r\n    float2 texelRadius = norm * TexelSize;\r\n    \r\n\r\n    //float2 Offset = ((float2(y, y)+Dither) * Multiplier) * TexelSize * direction[sd] ;\r\n    //float2 Offset = texelRadius * RadiusPixels * direction[sd] ;\r\n    float2 Offset = direction[sd] * aspect[sd] * norm * percent;\r\n    //float2 Offset = direction[sd]  * norm * percent;\r\n\r\n    //float depthMask = UtilityFunctions::mapRangeNormalized(SceneTextureLookup(UVDepth + Offset, CustomDepthTextureID, 0).r, minDepth, maxDepth);\r\n    //float stencil = SceneTextureLookup(UV2 + Offset, 25, true).r;\r\n    // if(stencil > 0){\r\n    //     continue;\r\n    // }\r\n    Offset *= mask * depthMask;\r\n    //float2 Offset = UV + (float2(y-1, y-1)* direction[sd]  * TexelSize * Multiplier);\r\n\r\n    float3 PixelColor = SceneTextureLookupCommon(Parameters, UV + Offset, SceneTextureID, 1).rgb;\r\n    //float stencil = SceneTextureLookup(UV2 + Offset, 25, true).r;\r\n\r\n    //float Weight = exp(-0.5 * pow(3.14159 * (x / Radius), 2)) * exp(-0.5 * pow(3.14159 * (y / Radius), 2));\r\n    float Weight = exp(-0.5 * pow(abs(3.14159 * norm), 2));\r\n\r\n    PixelSum += PixelColor * Weight;\r\n    WeightSum += Weight; \r\n    \r\n//}\r\n}\r\n\r\n//return float3(depthMask, depthMask, depthMask);\r\n//return PixelColorOrig;\r\nreturn PixelSum / WeightSum;\r\n"
   inputs [
    {
     inputName=Samples
     input {expression=&self:MaterialExpressionScalarParameter_3}
    }
    {
     inputName=RadiusPixels
     input {expression=&self:MaterialExpressionScalarParameter_4}
    }
    {
     inputName=Dither
     input {expression=&self:MaterialExpressionMaterialFunctionCall_8}
    }
    {
     inputName=SceneTexture
     input {
      expression=&self:MaterialExpressionSceneTexture_2
      mask=1
      maskA=1
      maskB=1
      maskG=1
      maskR=1
     }
    }
    {
     inputName=switchDirection
     input {expression=&self:MaterialExpressionStaticSwitchParameter_0}
    }
    {
     inputName=mask
     input {expression=&self:MaterialExpressionMaterialFunctionCall_10}
    }
    {
     inputName=minDepth
     input {expression=&self:MaterialExpressionScalarParameter_0}
    }
    {
     inputName=maxDepth
     input {expression=&self:MaterialExpressionScalarParameter_1}
    }
    {
     inputName=percent
     input {expression=&self:MaterialExpressionScalarParameter_2}
    }
   ]
  }
  ShadingPathSwitch(Default=@n3, Mobile=@MaterialExpressionSceneTexture_2.Color) {desc=Migrated from FeatureLevelSwitch}
  SmoothStep(Value=@MaterialExpressionMaterialFunctionCall_10.Result)
  ViewProperty() {property=MEVP_ViewSize}
 ]
}
