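# Sound cue for the impact when the Toggle Zoom DMR weapon is dropped.
# A random node picks one of three equally weighted drop impact waves and feeds a modulator that slightly varies pitch upward.
# The cue uses the weapon drop attenuation and loot collect sound class.
# A concurrency override caps simultaneous instances at a small count and prevents new plays beyond it.
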SoundCue /ToggleZoomDMRGameplay/Items/ToggleZoomDMR/Sounds/DMR_Drop_Impact_Cue
root=Modulator_0
nodes [
 {id=Modulator_0, children [Random_0], properties {pitchMax=1.1, pitchMin=1}}
 {id=Random_0, children [WavePlayer_0,WavePlayer_1,WavePlayer_2], properties {weights [1,1,1]}}
 {id=WavePlayer_0, properties {soundWaveAssetPtr=&/ToggleZoom_DMR_DropImpact_01}}
 {id=WavePlayer_1, properties {soundWaveAssetPtr=&/ToggleZoom_DMR_DropImpact_02}}
 {id=WavePlayer_2, properties {soundWaveAssetPtr=&/ToggleZoom_DMR_DropImpact_03}}
]
properties {
 attenuationSettings=&Fort_Weapon_Drop_Att
 bOverrideConcurrency=true
 maxDistance=3000
 soundClassObject=&Fort_Loot_Collect
 volumeMultiplier=0.9
 attenuationOverrides {falloffDistance=1000, attenuationShapeExtents {x=0}}
 concurrencyOverrides {maxCount=3, resolutionRule=PreventNew}
}
