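# First-person fire sound cue for the ToggleZoom DMR, routed through the first-person ranged weapon fire sound class and gunfire concurrency.
# Each shot mixes a randomized fire sample with a slightly delayed mechanical layer and a later, quieter casing impact, varied by modulators.
# A quality switch gives its first two levels larger sample pools plus a pitched sub layer.
# The last two levels use smaller pools without the sub.
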SoundCue /ToggleZoomDMRGameplay/Items/ToggleZoomDMR/Sounds/DMR_Fire_1P_Cue
root=QualityLevel_0
nodes [
 {id=Delay_1, children [Modulator_2], properties {delayMax=0.04, delayMin=0.04}}
 {id=Delay_2, children [Modulator_3], properties {delayMax=0.5, delayMin=0.3}}
 {id=Delay_3, children [Modulator_5], properties {delayMax=0.04, delayMin=0.04}}
 {id=Delay_4, children [Modulator_6], properties {delayMax=0.5, delayMin=0.3}}
 {id=Mixer_0, children [Modulator_4,Delay_3,Modulator_7,Delay_4], properties {inputVolume [1.1,1.1,0.6,1]}}
 {id=Mixer_1, children [Modulator_1,Delay_1,Delay_2], properties {inputVolume [1.1,1.1,1]}}
 {id=Modulator_1, children [Random_0], properties {volumeMax=1.7, volumeMin=1.5}}
 {id=Modulator_2, children [Random_2], properties {pitchMin=0.9, volumeMax=1.25, volumeMin=1.05}}
 {id=Modulator_3, children [Random_3], properties {pitchMax=1.1, pitchMin=0.9, volumeMax=0.35, volumeMin=0.3}}
 {id=Modulator_4, children [Random_4], properties {volumeMax=1.7, volumeMin=1.5}}
 {id=Modulator_5, children [Random_6], properties {pitchMin=0.9, volumeMax=1.25, volumeMin=1.05}}
 {id=Modulator_6, children [Random_7], properties {pitchMax=1.1, pitchMin=0.9, volumeMax=0.35, volumeMin=0.3}}
 {id=Modulator_7, children [WavePlayer_31], properties {pitchMax=1.07, pitchMin=1.07}}
 {id=QualityLevel_0, children [Mixer_0,Mixer_0,Mixer_1,Mixer_1]}
 {id=Random_0, children [WavePlayer_0,WavePlayer_1,WavePlayer_3], properties {weights [1,1,1]}}
 {id=Random_2, children [WavePlayer_8,WavePlayer_9], properties {weights [1,1]}}
 {id=Random_3, children [WavePlayer_21, WavePlayer_22], properties {weights [1,1]}}
 {
  id=Random_4
  children [WavePlayer_2, WavePlayer_6, WavePlayer_12, WavePlayer_13, WavePlayer_14, WavePlayer_15]
  properties {weights [1,1,1,1,1,1]}
 }
 {
  id=Random_6
  children [WavePlayer_16, WavePlayer_17, WavePlayer_24, WavePlayer_25, WavePlayer_26, WavePlayer_27]
  properties {weights [1,1,1,1,1,1]}
 }
 {id=Random_7, children [WavePlayer_34, WavePlayer_35, WavePlayer_36, WavePlayer_44, WavePlayer_45], properties {weights [1,1,1,1,1]}}
 {id=WavePlayer_0, properties {soundWaveAssetPtr=&/ToggleZoom_DMR_Fire_1P_01}}
 {id=WavePlayer_1, properties {soundWaveAssetPtr=&/ToggleZoom_DMR_Fire_1P_02}}
 {id=WavePlayer_12, properties {soundWaveAssetPtr=&/ToggleZoom_DMR_Fire_1P_03}}
 {id=WavePlayer_13, properties {soundWaveAssetPtr=&/ToggleZoom_DMR_Fire_1P_04}}
 {id=WavePlayer_14, properties {soundWaveAssetPtr=&/ToggleZoom_DMR_Fire_1P_05}}
 {id=WavePlayer_15, properties {soundWaveAssetPtr=&/ToggleZoom_DMR_Fire_1P_06}}
 {id=WavePlayer_16, properties {soundWaveAssetPtr=&/ToggleZoom_DMR_Fire_1P_Mech_01}}
 {id=WavePlayer_17, properties {soundWaveAssetPtr=&/ToggleZoom_DMR_Fire_1P_Mech_02}}
 {id=WavePlayer_2, properties {soundWaveAssetPtr=&ToggleZoom_DMR_Fire_1P_01}}
 {id=WavePlayer_21, properties {soundWaveAssetPtr=&/ToggleZoom_DMR_Casing_Impact_01}}
 {id=WavePlayer_22, properties {soundWaveAssetPtr=&/ToggleZoom_DMR_Casing_Impact_02}}
 {id=WavePlayer_24, properties {soundWaveAssetPtr=&/ToggleZoom_DMR_Fire_1P_Mech_03}}
 {id=WavePlayer_25, properties {soundWaveAssetPtr=&/ToggleZoom_DMR_Fire_1P_Mech_04}}
 {id=WavePlayer_26, properties {soundWaveAssetPtr=&/ToggleZoom_DMR_Fire_1P_Mech_05}}
 {id=WavePlayer_27, properties {soundWaveAssetPtr=&/ToggleZoom_DMR_Fire_1P_Mech_06}}
 {id=WavePlayer_3, properties {soundWaveAssetPtr=&ToggleZoom_DMR_Fire_1P_03}}
 {id=WavePlayer_31, properties {soundWaveAssetPtr=&/ToggleZoom_DMR_Fire_1P_Sub_04}}
 {id=WavePlayer_34, properties {soundWaveAssetPtr=&ToggleZoom_DMR_Casing_Impact_01}}
 {id=WavePlayer_35, properties {soundWaveAssetPtr=&ToggleZoom_DMR_Casing_Impact_02}}
 {id=WavePlayer_36, properties {soundWaveAssetPtr=&/ToggleZoom_DMR_Casing_Impact_03}}
 {id=WavePlayer_44, properties {soundWaveAssetPtr=&/ToggleZoom_DMR_Casing_Impact_04}}
 {id=WavePlayer_45, properties {soundWaveAssetPtr=&/ToggleZoom_DMR_Casing_Impact_05}}
 {id=WavePlayer_6, properties {soundWaveAssetPtr=&ToggleZoom_DMR_Fire_1P_02}}
 {id=WavePlayer_8, properties {soundWaveAssetPtr=&ToggleZoom_DMR_Fire_1P_Mech_01}}
 {id=WavePlayer_9, properties {soundWaveAssetPtr=&ToggleZoom_DMR_Fire_1P_Mech_02}}
]
properties {
 attenuationSettings=&Fort_Weapon_Projectile_FirstPerson_Athena_Att
 bDebug=true
 bExcludeFromRandomNodeBranchCulling=true
 maxDistance=2097152
 soundClassObject=&Fort_WeaponRanged_Fire_1P_PSM
 volumeMultiplier=0.7
 attenuationOverrides {bSpatialize=false}
 concurrencyOverrides {bLimitToOwner=true, maxCount=3, resolutionRule=StopOldest}
 concurrencySet [&Weapon_Gunfire_FirstPerson_Conc]
}
