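# Third-person fire sound cue for the ToggleZoom DMR, using global and local gunfire concurrency sets.
# A quality switch sends the first two levels to a mix with larger wave pools and the last two to a mix with fewer variations.
# Each mix layers close, medium and distant shots, each randomly picked, modulated, then given its own attenuation and a sound class override.
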SoundCue /ToggleZoomDMRGameplay/Items/ToggleZoomDMR/Sounds/DMR_Fire_3P_Cue
root=QualityLevel_1
nodes [
 {
  id=Attenuation_0
  children [Modulator_6]
  properties {attenuationSettings=&Fort_Weapon_Projectile_Distant_Mono_Athena_Att}
 }
 {id=Attenuation_1, children [Modulator_0], properties {attenuationSettings=&Fort_Weapon_Projectile_Close_Athena_Att}}
 {
  id=Attenuation_2
  children [Modulator_2]
  properties {attenuationSettings=&Fort_Weapon_Projectile_Medium_Mono_Athena_Att}
 }
 {id=Attenuation_3, children [Modulator_1], properties {attenuationSettings=&Fort_Weapon_Projectile_Close_Athena_Att}}
 {
  id=Attenuation_4
  children [Modulator_3]
  properties {attenuationSettings=&Fort_Weapon_Projectile_Distant_Mono_Athena_Att}
 }
 {
  id=Attenuation_5
  children [Modulator_5]
  properties {attenuationSettings=&Fort_Weapon_Projectile_Medium_Mono_Athena_Att}
 }
 {id=Mixer_0, children [SoundClass_1,SoundClass_2,SoundClass_4], properties {inputVolume [0.85,0.8,0.7]}}
 {id=Mixer_3, children [SoundClass_3,SoundClass_5,SoundClass_0], properties {inputVolume [0.85,0.8,0.7]}}
 {id=Modulator_0, children [Random_0], properties {pitchMax=1, pitchMin=0.9, volumeMax=0.9, volumeMin=0.8}}
 {id=Modulator_1, children [Random_5], properties {pitchMax=1, pitchMin=0.9, volumeMax=0.9, volumeMin=0.8}}
 {id=Modulator_2, children [Random_1], properties {volumeMax=1.2, volumeMin=1.1}}
 {id=Modulator_3, children [Random_2], properties {pitchMax=0.95, pitchMin=0.85, volumeMax=0.6, volumeMin=0.5}}
 {id=Modulator_5, children [Random_6], properties {volumeMax=1.2, volumeMin=1.1}}
 {id=Modulator_6, children [Random_7], properties {pitchMax=0.95, pitchMin=0.85, volumeMax=0.6, volumeMin=0.5}}
 {id=QualityLevel_1, children [Mixer_3,Mixer_3,Mixer_0,Mixer_0]}
 {id=Random_0, children [WavePlayer_0,WavePlayer_1,WavePlayer_2], properties {weights [1,1,1]}}
 {id=Random_1, children [WavePlayer_6,WavePlayer_7], properties {weights [1,1]}}
 {id=Random_2, children [WavePlayer_11, WavePlayer_12], properties {bShouldExcludeFromBranchCulling=true, weights [1,1]}}
 {
  id=Random_5
  children [WavePlayer_14, WavePlayer_25, WavePlayer_26, WavePlayer_27, WavePlayer_28, WavePlayer_29]
  properties {weights [1,1,1,1,1,1]}
 }
 {id=Random_6, children [WavePlayer_30, WavePlayer_31, WavePlayer_32, WavePlayer_33, WavePlayer_34], properties {weights [1,1,1,1,1]}}
 {id=Random_7, children [WavePlayer_35, WavePlayer_36, WavePlayer_37, WavePlayer_38, WavePlayer_39], properties {weights [1,1,1,1,1]}}
 {
  id=SoundClass_0
  children [Attenuation_0]
  properties {soundClassOverride=&Fort_WeaponRanged_Fire_Distant}
 }
 {
  id=SoundClass_1
  children [Attenuation_1]
  properties {soundClassOverride=&Fort_WeaponRanged_Fire_3P_PSM}
 }
 {
  id=SoundClass_2
  children [Attenuation_2]
  properties {soundClassOverride=&Fort_WeaponRanged_Fire_Distant}
 }
 {
  id=SoundClass_3
  children [Attenuation_3]
  properties {soundClassOverride=&Fort_WeaponRanged_Fire_3P_PSM}
 }
 {
  id=SoundClass_4
  children [Attenuation_4]
  properties {soundClassOverride=&Fort_WeaponRanged_Fire_Distant}
 }
 {
  id=SoundClass_5
  children [Attenuation_5]
  properties {soundClassOverride=&Fort_WeaponRanged_Fire_Distant}
 }
 {id=WavePlayer_0, properties {soundWaveAssetPtr=&/ToggleZoom_DMR_Fire_Close_01}}
 {id=WavePlayer_1, properties {soundWaveAssetPtr=&/ToggleZoom_DMR_Fire_Close_02}}
 {id=WavePlayer_11, properties {soundWaveAssetPtr=&/ToggleZoom_DMR_Fire_Distant_01}}
 {id=WavePlayer_12, properties {soundWaveAssetPtr=&/ToggleZoom_DMR_Fire_Distant_02}}
 {id=WavePlayer_14, properties {soundWaveAssetPtr=&ToggleZoom_DMR_Fire_Close_01}}
 {id=WavePlayer_2, properties {soundWaveAssetPtr=&/ToggleZoom_DMR_Fire_Close_03}}
 {id=WavePlayer_25, properties {soundWaveAssetPtr=&ToggleZoom_DMR_Fire_Close_02}}
 {id=WavePlayer_26, properties {soundWaveAssetPtr=&ToggleZoom_DMR_Fire_Close_03}}
 {id=WavePlayer_27, properties {soundWaveAssetPtr=&/ToggleZoom_DMR_Fire_Close_04}}
 {id=WavePlayer_28, properties {soundWaveAssetPtr=&/ToggleZoom_DMR_Fire_Close_05}}
 {id=WavePlayer_29, properties {soundWaveAssetPtr=&/ToggleZoom_DMR_Fire_Close_06}}
 {id=WavePlayer_30, properties {soundWaveAssetPtr=&/ToggleZoom_DMR_Fire_Medium_01}}
 {id=WavePlayer_31, properties {soundWaveAssetPtr=&/ToggleZoom_DMR_Fire_Medium_02}}
 {id=WavePlayer_32, properties {soundWaveAssetPtr=&/ToggleZoom_DMR_Fire_Medium_03}}
 {id=WavePlayer_33, properties {soundWaveAssetPtr=&/ToggleZoom_DMR_Fire_Medium_04}}
 {id=WavePlayer_34, properties {soundWaveAssetPtr=&/ToggleZoom_DMR_Fire_Medium_05}}
 {id=WavePlayer_35, properties {soundWaveAssetPtr=&ToggleZoom_DMR_Fire_Distant_01}}
 {id=WavePlayer_36, properties {soundWaveAssetPtr=&ToggleZoom_DMR_Fire_Distant_02}}
 {id=WavePlayer_37, properties {soundWaveAssetPtr=&/ToggleZoom_DMR_Fire_Distant_03}}
 {id=WavePlayer_38, properties {soundWaveAssetPtr=&/ToggleZoom_DMR_Fire_Distant_04}}
 {id=WavePlayer_39, properties {soundWaveAssetPtr=&/ToggleZoom_DMR_Fire_Distant_05}}
 {id=WavePlayer_6, properties {soundWaveAssetPtr=&ToggleZoom_DMR_Fire_Medium_01}}
 {id=WavePlayer_7, properties {soundWaveAssetPtr=&ToggleZoom_DMR_Fire_Medium_02}}
]
properties {
 bDebug=true
 bExcludeFromRandomNodeBranchCulling=true
 maxDistance=27500
 volumeMultiplier=1
 concurrencyOverrides {bLimitToOwner=true, maxCount=3, resolutionRule=StopOldest}
 concurrencySet [&Weapon_Gunfire_Global_Conc, &Weapon_Gunfire_Local_Conc]
}
