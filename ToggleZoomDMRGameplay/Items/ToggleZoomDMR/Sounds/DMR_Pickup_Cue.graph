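# Sound cue for picking up the Toggle Zoom DMR weapon.
# A random node picks one of three equally weighted pickup waves, and a modulator slightly varies its volume.
# The cue uses the shared weapon pickup attenuation and loot collect sound class, and references a pickup concurrency asset to limit simultaneous plays.
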SoundCue /ToggleZoomDMRGameplay/Items/ToggleZoomDMR/Sounds/DMR_Pickup_Cue
root=Modulator_0
nodes [
 {id=Modulator_0, children [Random_0], properties {volumeMax=0.9, volumeMin=0.8}}
 {id=Random_0, children [WavePlayer_0,WavePlayer_1,WavePlayer_3], properties {weights [1,1,1]}}
 {id=WavePlayer_0, properties {soundWaveAssetPtr=&/ToggleZoom_DMR_Pickup_01}}
 {id=WavePlayer_1, properties {soundWaveAssetPtr=&/ToggleZoom_DMR_Pickup_02}}
 {id=WavePlayer_3, properties {soundWaveAssetPtr=&/ToggleZoom_DMR_Pickup_03}}
]
properties {
 attenuationSettings=&Fort_Weapon_Pickup_Att
 maxDistance=3500
 soundClassObject=&Fort_Loot_Collect
 volumeMultiplier=0.7
 concurrencyOverrides {maxCount=1, resolutionRule=PreventNew}
 concurrencySet [&PickupMaximum_Oldest]
}
