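# Sound cue for the end of a reload on the ToggleZoom DMR weapon.
# A random node picks one of four equally weighted reload-end waves and feeds a modulator that slightly varies and lowers the volume.
# The cue plays slightly pitched up in the weapon foley sound class with weapon reload attenuation, and it overrides concurrency and has its debug flag on.
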SoundCue /ToggleZoomDMRGameplay/Items/ToggleZoomDMR/Sounds/DMR_Reload_End_Cue
root=Modulator_1
nodes [
 {id=Modulator_1, children [Random_0], properties {volumeMax=0.8, volumeMin=0.7}}
 {id=Random_0, children [WavePlayer_0,WavePlayer_2,WavePlayer_3,WavePlayer_4], properties {weights [1,1,1,1]}}
 {id=WavePlayer_0, properties {soundWaveAssetPtr=&/ToggleZoom_DMR_Reload_End_01}}
 {id=WavePlayer_2, properties {soundWaveAssetPtr=&/ToggleZoom_DMR_Reload_End_02}}
 {id=WavePlayer_3, properties {soundWaveAssetPtr=&/ToggleZoom_DMR_Reload_End_03}}
 {id=WavePlayer_4, properties {soundWaveAssetPtr=&/ToggleZoom_DMR_Reload_End_04}}
]
properties {
 attenuationSettings=&Fort_Weapon_Reload_Att
 bDebug=true
 bOverrideConcurrency=true
 maxDistance=2250
 pitchMultiplier=1.1
 soundClassObject=&Fort_WeaponFoley
 volumeMultiplier=0.852
}
