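# Sound cue for the magazine or round insert foley during the toggle-zoom DMR's reload.
# A Random node picks one of four equally weighted insert waves and feeds a root Modulator that pins pitch at normal.
# It plays in the weapon foley sound class at reduced volume, with shared reload attenuation and a concurrency override.
# The cue's debug flag is left on.
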SoundCue /ToggleZoomDMRGameplay/Items/ToggleZoomDMR/Sounds/DMR_Reload_Insert_Cue
root=Modulator_1
nodes [
 {id=Modulator_1, children [Random_0], properties {pitchMax=1, pitchMin=1}}
 {id=Random_0, children [WavePlayer_1,WavePlayer_2,WavePlayer_3,WavePlayer_4], properties {weights [1,1,1,1]}}
 {id=WavePlayer_1, properties {soundWaveAssetPtr=&/ToggleZoom_DMR_Reload_Insert_01}}
 {id=WavePlayer_2, properties {soundWaveAssetPtr=&/ToggleZoom_DMR_Reload_Insert_02}}
 {id=WavePlayer_3, properties {soundWaveAssetPtr=&/ToggleZoom_DMR_Reload_Insert_03}}
 {id=WavePlayer_4, properties {soundWaveAssetPtr=&/ToggleZoom_DMR_Reload_Insert_04}}
]
properties {
 attenuationSettings=&Fort_Weapon_Reload_Att
 bDebug=true
 bOverrideConcurrency=true
 maxDistance=2250
 soundClassObject=&Fort_WeaponFoley
 volumeMultiplier=0.652
}
