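# Sound cue for the start of a reload on the ToggleZoom DMR, layering reload foley with a delayed magazine impact.
# A mixer combines a random reload-start sound with a quieter random magazine impact that plays after a fixed delay of about half a second.
# Each layer picks from three equally weighted waves and passes through its own modulator.
# It plays in the weapon foley sound class with weapon reload attenuation.
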SoundCue /ToggleZoomDMRGameplay/Items/ToggleZoomDMR/Sounds/DMR_Reload_Start_Cue
root=Mixer_1
nodes [
 {id=Delay_0, children [Modulator_0], properties {delayMax=0.55, delayMin=0.55}}
 {id=Mixer_1, children [Modulator_1,Delay_0], properties {inputVolume [1,0.5]}}
 {id=Modulator_0, children [Random_1]}
 {id=Modulator_1, children [Random_0]}
 {id=Random_0, children [WavePlayer_1,WavePlayer_2,WavePlayer_3], properties {weights [1,1,1]}}
 {id=Random_1, children [WavePlayer_0,WavePlayer_4,WavePlayer_5], properties {weights [1,1,1]}}
 {id=WavePlayer_0, properties {soundWaveAssetPtr=&/ToggleZoom_DMR_Magazine_Impact_01}}
 {id=WavePlayer_1, properties {soundWaveAssetPtr=&/ToggleZoom_DMR_Reload_Start_01}}
 {id=WavePlayer_2, properties {soundWaveAssetPtr=&/ToggleZoom_DMR_Reload_Start_02}}
 {id=WavePlayer_3, properties {soundWaveAssetPtr=&/ToggleZoom_DMR_Reload_Start_03}}
 {id=WavePlayer_4, properties {soundWaveAssetPtr=&/ToggleZoom_DMR_Magazine_Impact_02}}
 {id=WavePlayer_5, properties {soundWaveAssetPtr=&/ToggleZoom_DMR_Magazine_Impact_03}}
]
properties {
 attenuationSettings=&Fort_Weapon_Reload_Att
 bDebug=true
 bOverrideConcurrency=true
 maxDistance=2250
 soundClassObject=&Fort_WeaponFoley
 volumeMultiplier=0.802
}
