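# Sound cue for the scope zoom-in toggle on the Toggle Zoom DMR, classed as weapon foley.
# A random node picks one of two equally weighted zoom-in wave variations, and a modulator adds slight volume variance.
# It overrides a first-person projectile attenuation asset with its own short-range falloff and no reverb send, and overrides concurrency to prevent new instances at the limit.
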SoundCue /ToggleZoomDMRGameplay/Items/ToggleZoomDMR/Sounds/DMR_Scope_Toggle_ZoomIn_Cue
root=Modulator_0
nodes [
 {id=Modulator_0, children [Random_2], properties {volumeMax=1.5, volumeMin=1.4}}
 {id=Random_2, children [WavePlayer_3,WavePlayer_8], properties {weights [1,1]}}
 {id=WavePlayer_3, properties {soundWaveAssetPtr=&/ToggleZoom_DMR_Scope_Toggle_In_01}}
 {id=WavePlayer_8, properties {soundWaveAssetPtr=&/ToggleZoom_DMR_Scope_Toggle_In_02}}
]
properties {
 attenuationSettings=&Fort_Weapon_Projectile_FirstPerson_Athena_Att
 bOverrideAttenuation=true
 bOverrideConcurrency=true
 maxDistance=2000
 soundClassObject=&Fort_WeaponFoley
 volumeMultiplier=0.251
 attenuationOverrides {
  bEnableReverbSend=false
  falloffDistance=1200
  nonSpatializedRadiusEnd=300
  nonSpatializedRadiusStart=300
  stereoSpread=2000
  attenuationShapeExtents {x=800}
 }
 concurrencyOverrides {resolutionRule=PreventNew}
}
