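# Sound cue for the DMR scope's zoom-out toggle, classed as weapon foley.
# It randomly picks one of two equally weighted waves, passes it through a slight volume variation, and starts it after a short fixed delay.
# Attenuation and concurrency are overridden on the cue, with concurrency set to prevent new sounds once its limit is reached.
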SoundCue /ToggleZoomDMRGameplay/Items/ToggleZoomDMR/Sounds/DMR_Scope_Toggle_ZoomOut_Cue
root=Delay_0
nodes [
 {id=Delay_0, children [Modulator_0], properties {delayMax=0.1, delayMin=0.1}}
 {id=Modulator_0, children [Random_2], properties {volumeMax=1.5, volumeMin=1.4}}
 {id=Random_2, children [WavePlayer_1,WavePlayer_6], properties {weights [1,1]}}
 {id=WavePlayer_1, properties {soundWaveAssetPtr=&/ToggleZoom_DMR_Scope_Toggle_Out_01}}
 {id=WavePlayer_6, properties {soundWaveAssetPtr=&/ToggleZoom_DMR_Scope_Toggle_Out_02}}
]
properties {
 attenuationSettings=&Fort_Weapon_Projectile_FirstPerson_Athena_Att
 bOverrideAttenuation=true
 bOverrideConcurrency=true
 maxDistance=2700
 soundClassObject=&Fort_WeaponFoley
 volumeMultiplier=0.301
 attenuationOverrides {falloffDistance=1500, nonSpatializedRadiusStart=200, attenuationShapeExtents {x=1200}}
 concurrencyOverrides {resolutionRule=PreventNew}
}
