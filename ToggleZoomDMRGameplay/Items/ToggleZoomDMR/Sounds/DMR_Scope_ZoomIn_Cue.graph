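# Sound cue for the scope zoom-in on the toggle-zoom DMR, played as quiet weapon foley.
# A mixer layers one of three random scope-in samples, boosted by a fixed modulator, over a quieter random player-run foley clip with slight pitch and volume variation.
# It overrides attenuation with a short range and no reverb send, and its concurrency rule blocks new instances once the limit is reached.
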SoundCue /ToggleZoomDMRGameplay/Items/ToggleZoomDMR/Sounds/DMR_Scope_ZoomIn_Cue
root=Mixer_1
nodes [
 {id=Mixer_1, children [Modulator_2,Modulator_3], properties {inputVolume [1,0.4]}}
 {id=Modulator_2, children [Random_0], properties {pitchMax=1, pitchMin=1, volumeMax=1.5, volumeMin=1.5}}
 {id=Modulator_3, children [Random_1], properties {pitchMax=0.95, pitchMin=1.05, volumeMax=0.9, volumeMin=0.7}}
 {id=Random_0, children [WavePlayer_0,WavePlayer_1,WavePlayer_2], properties {weights [1,1,1]}}
 {id=Random_1, children [WavePlayer_7,WavePlayer_6,WavePlayer_5,WavePlayer_4], properties {weights [1,1,1,1]}}
 {id=WavePlayer_0, properties {soundWaveAssetPtr=&/ToggleZoom_DMR_Scope_In_01}}
 {id=WavePlayer_1, properties {soundWaveAssetPtr=&/ToggleZoom_DMR_Scope_In_02}}
 {id=WavePlayer_2, properties {soundWaveAssetPtr=&/ToggleZoom_DMR_Scope_In_03}}
 {id=WavePlayer_4, properties {soundWaveAssetPtr=/Game/Sounds/Fort_Foley_Sounds/Moves/player_run_foley_12}}
 {id=WavePlayer_5, properties {soundWaveAssetPtr=/Game/Sounds/Fort_Foley_Sounds/Moves/player_run_foley_10}}
 {id=WavePlayer_6, properties {soundWaveAssetPtr=/Game/Sounds/Fort_Foley_Sounds/Moves/player_run_foley_03}}
 {id=WavePlayer_7, properties {soundWaveAssetPtr=/Game/Sounds/Fort_Foley_Sounds/Moves/player_run_foley_09}}
]
properties {
 attenuationSettings=&Fort_Weapon_Projectile_FirstPerson_Athena_Att
 bOverrideAttenuation=true
 bOverrideConcurrency=true
 maxDistance=2000
 soundClassObject=&Fort_WeaponFoley
 volumeMultiplier=0.175
 attenuationOverrides {
  bEnableReverbSend=false
  falloffDistance=1200
  nonSpatializedRadiusEnd=300
  nonSpatializedRadiusStart=300
  stereoSpread=2000
  attenuationShapeExtents {x=800}
 }
 concurrencyOverrides {resolutionRule=PreventNew}
}
