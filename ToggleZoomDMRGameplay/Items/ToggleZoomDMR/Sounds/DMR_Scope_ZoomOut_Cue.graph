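# Sound cue that plays when the Toggle Zoom DMR scope zooms out, classed as weapon foley.
# A mixer layers two randomized branches: one of three scope-out waves with raised, varied volume, and a quieter run-foley wave with slightly raised, varied pitch.
# The low-volume cue references first-person projectile attenuation with overridden falloff, and a concurrency override that prevents new instances at the limit.
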SoundCue /ToggleZoomDMRGameplay/Items/ToggleZoomDMR/Sounds/DMR_Scope_ZoomOut_Cue
root=Mixer_1
nodes [
 {id=Mixer_1, children [Modulator_3,Modulator_2], properties {inputVolume [1.2,0.6]}}
 {id=Modulator_2, children [Random_0], properties {pitchMax=1.2, pitchMin=1}}
 {id=Modulator_3, children [Random_1], properties {volumeMax=1.9, volumeMin=1.7}}
 {id=Random_0, children [WavePlayer_2,WavePlayer_0], properties {weights [1,1]}}
 {id=Random_1, children [WavePlayer_3,WavePlayer_4,WavePlayer_5], properties {weights [1,1,1]}}
 {id=WavePlayer_0, properties {soundWaveAssetPtr=/Game/Sounds/Fort_Foley_Sounds/Moves/player_run_foley_05}}
 {id=WavePlayer_2, properties {soundWaveAssetPtr=/Game/Sounds/Fort_Foley_Sounds/Moves/player_run_foley_01}}
 {id=WavePlayer_3, properties {soundWaveAssetPtr=&/ToggleZoom_DMR_Scope_Out_01}}
 {id=WavePlayer_4, properties {soundWaveAssetPtr=&/ToggleZoom_DMR_Scope_Out_02}}
 {id=WavePlayer_5, properties {soundWaveAssetPtr=&/ToggleZoom_DMR_Scope_Out_03}}
]
properties {
 attenuationSettings=&Fort_Weapon_Projectile_FirstPerson_Athena_Att
 bOverrideAttenuation=true
 bOverrideConcurrency=true
 maxDistance=2700
 soundClassObject=&Fort_WeaponFoley
 volumeMultiplier=0.125
 attenuationOverrides {falloffDistance=1500, nonSpatializedRadiusStart=200, attenuationShapeExtents {x=1200}}
 concurrencyOverrides {resolutionRule=PreventNew}
}
