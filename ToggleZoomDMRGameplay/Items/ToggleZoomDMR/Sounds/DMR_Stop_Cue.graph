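# DMR_Stop_Cue is a SoundCue for the ToggleZoomDMR weapon, named as its stop sound.
# It has no root node or sound nodes, so it plays no audio as authored.
# It still carries a long-range assault rifle attenuation asset, the third-person ranged weapon fire sound class, and local and global gunfire concurrency sets.
# Its debug flag is on.
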SoundCue /ToggleZoomDMRGameplay/Items/ToggleZoomDMR/Sounds/DMR_Stop_Cue
root=null
nodes []
properties {
 attenuationSettings=/Game/Sounds/Attenuation/Fort_Weapon_AR_Lng_Att
 bDebug=true
 maxDistance=4000
 soundClassObject=&Fort_WeaponRanged_Fire_3P_PSM
 concurrencyOverrides {maxCount=2}
 concurrencySet [&Weapon_Gunfire_Local_Conc, &Weapon_Gunfire_Global_Conc]
}
